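# Settings rotator widget for Fortnite settings entries, built on FortSettingsRotator, showing centered option text over a material-driven border.
# PreConstruct caches the border's dynamic material; option events feed it count, selected and default indices as scalar parameters, so the material receives all indicator state.
# Selection plays a press sound only when OptionsSet is true and the index differs from PreviousOption.
# SetShowBackgroundOption toggles the material's Opacity parameter.
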WidgetBlueprint /FortniteCoreUI/Settings/SettingEntries/EntryComponents/SettingsRotator
graphs [
 EventGraph {nodes [
   n1 = Event On Options Populated!() -> (OutputDelegate, then, Count:int=0) {is=title K2Node_Event}
   > SetScalarParameterValue(
    self=Get MID_Background() -> (MID_Background:Material Instance Dynamic Object Reference)
    ParameterName=Options
    Value="To Float (Integer)"(InInt=@n1.Count)
   )
   n2 = Event On Options Selected!() -> (OutputDelegate, then, Index:int=0) {is=title K2Node_Event}
   > SetScalarParameterValue(self=Get MID_Background(), ParameterName=SelectedOption, Value="To Float (Integer)"(InInt=@n2.Index))
   > n3 = Branch(Condition=Get OptionsSet() -> (OptionsSet:bool))
   > Branch(Condition="Not Equal (Integer)"(A=@n2.Index, B=Get PreviousOption() -> (PreviousOption:int=0)))
   > PlaySound2D(Sound=Get PressedSound() -> (PressedSound:Sound Base Object Reference))
   > Set PreviousOption(PreviousOption:int=@n2.Index) -> (Output_Get:int=0)
   Set PreviousOption(execute=@n3.else, PreviousOption=@n2.Index) -> (Output_Get=0)
   n4 = Event On Default Option Specified!() -> (OutputDelegate, then, DefaultOptionIndex:int=0) {is=title K2Node_Event}
   > SetScalarParameterValue(self=Get MID_Background(), ParameterName=DefaultOption, Value="To Float (Integer)"(InInt=@n4.DefaultOptionIndex))
   Event Construct()
   > Parent: Construct()
   Event PreConstruct()
   > n5 = GetDynamicMaterial(self:Border Object Reference=Get Background() -> (Background:Border Object Reference))
   > Set MID_Background(MID_Background:Material Instance Dynamic Object Reference=@n5) -> (Output_Get:Material Instance Dynamic Object Reference)
  ]}
 SetShowBackgroundOption {nodes [
   n1 = SetShowBackgroundOption!() -> (then, IsShowOptionBackground:bool=~) {is=title K2Node_FunctionEntry}
   > SetScalarParameterValue(self=Get MID_Background(), ParameterName=Opacity, Value=SelectFloat(A=1, B=0, bPickA=@n1))
  ]}
]
variables [
 AllowSound {type=boolean}
 MID_Background {type=MaterialInstanceDynamic}
 OptionsSet {type=boolean}
 PressedSound {type=SoundBase, default=/Game/Sounds/Fort_UI_Sounds/FrontEnd/DefaultSounds/PressedButtons/MSS_UI_Button_Pressed_Default}
 PreviousOption {type=integer}
]
widgets {
 info {inheritedWidgetCount=2, parentClass=/Script/FortniteUI.FortSettingsRotator, rootWidgetClass=&Border, widgetCount=3}
 widgets [
  Background {
   widgetClassPath=&Border
   bIsVariable=true
   widget=Background
   properties {
    horizontalAlignment=HAlign_Center
    verticalAlignment=VAlign_Center
    background {
     bIsDynamicallyLoaded=false
     drawAs=Image
     imageType=NoImage
     mirroring=NoMirror
     resourceName=None
     resourceObject=/FortniteCoreUI/Settings/Materials/MI_SettingsRotator
     tiling=NoTile
     imageSize=<32 32>
     margin=<ltrb 0.5 0 0.5 0>
     outlineSettings {
      bUseBrushTransparency=false
      roundingType=HalfHeightRadius
      width=0
      color {colorUseRule=UseColor_Specified, specifiedColor=<rgba 0 0 0 0>}
      cornerRadii=<0 0 0 0>
     }
     tintColor {colorUseRule=UseColor_Specified, specifiedColor=<rgba 1 1 1 1>}
     uVRegion {bIsValid=false, max=<0 0>, min=<0 0>}
    }
    padding=<ltrb 0 0 0 0>
   }
  }
  MyText {
   widgetClassPath=&CommonTextBlock
   bInherited=true
   bIsVariable=true
   parent=Background
   widget=MyText
   properties {
    justification=Center
    scrollStyle=&TextScrollStyle-Base_C
    text="{Option}"
    colorAndOpacity {colorUseRule=UseColor_Specified, specifiedColor=<rgba 0.75 0.75 0.75 1>}
    font {
     bForceMonospaced=false
     bMaterialIsStencil=false
     fontMaterial=null
     fontObject=&HeadingNow
     letterSpacing=0
     monospacedWidth=1
     size=18
     skewAmount=0
     typefaceFontName=74 Regular
     outlineSettings {
      bApplyOutlineToDropShadows=false
      bMiteredCorners=false
      bSeparateFillAlpha=false
      outlineMaterial=null
      outlineSize=0
      outlineColor=<rgba 0 0 0 1>
     }
    }
    shadowOffset=<0 0>
   }
   slot {class=BorderSlot, properties {horizontalAlignment=HAlign_Center, verticalAlignment=VAlign_Center, padding=<ltrb 0 0 0 0>}}
  }
  InputActionWidget {widgetClassPath=&CommonActionWidget, bInherited=true, widget=null}
 ]
}
